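# Translucent lit Niagara mesh-particle material for the ice effect on the HistorianMale pickaxe.
# A smoke tile texture, with parameterized UV tiling and constant offsets fed through Panner time inputs, is tinted by EmissiveColor and particle color, then passed through an emissive-adjustment function.
# The alpha chain multiplies that texture by a second-UV vertical gradient, particle alpha and a SmoothStep dissolve driven by a dynamic parameter and hardness scalar.
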Material /BRCosmetics/Weapons/FORT_Melee/Pickaxe_HistorianMale/FX/M_Pickaxe_HistorianMale_Ice
graph {
 properties {
  bUsedWithNiagaraMeshParticles=true
  blendMode=BLEND_Translucent
  pixelDepthOffsetMode=PDOM_Legacy
  previewMesh=&/SM_Pickaxe_HistorianMale_Ice
  usedShadingModels=Default Lit
 }
 nodes [
  n1 = VectorParameter() {parameterName=EmissiveColor, defaultValue=<rgba 0.3 0.65 1 1>}
  n2 = TextureCoordinate()
  n3 = VectorParameter() {
   parameterName=EmissiveUV
   channelNames {a=EmissiveOffset V, b=EmissiveOffset U, g=EmissiveTiling V, r=EmissiveTiling U}
   defaultValue=<rgba 1 1 0.4 0>
  }
  n4 = AppendVector(
   A=Multiply(A=ComponentMask(None=@n2) {r=true}, B=@n3.EmissiveTiling U)
   B=Multiply(A=ComponentMask(None=@n2) {g=true}, B=@n3.EmissiveTiling V)
  )
  n5 = TextureCoordinate() {uTiling=0, vTiling=0}
  n6 = TextureSample(UVs=Add(A=Panner(Coordinate=@n4, Time=@n3.EmissiveOffset U) {speedX=1}, B=Panner(Coordinate=@n5, Time=@n3.EmissiveOffset V) {speedY=1})) {texture=/Game/Global/Textures/Default/Tileable_Patterns/T_Smoke_Puff_Tile_Up}
  n7 = ParticleColor()
  MaterialFunctionCall(Emissive=Multiply(A=Multiply(A=@n1.RGB, B=@n6.RGB), B=@n7.RGB)) {materialFunction=&MF_FX_EmissiveAdjustments}
  n8 = TextureCoordinate() {coordinateIndex=1}
  n9 = Subtract(A=OneMinus(None=ComponentMask(None=@n8) {g=true})) {constB=0.1}
  n10 = DynamicParameter() {defaultValue=<rgba 0 1 1 1>, paramNames [Dissolve,Param2,Param3,Param4]}
  n11 = Add(A=@n10.Dissolve, B=ScalarParameter() {defaultValue=0.7, parameterName=Dissolve Hardness})
  Multiply(
   A=@n7.A
   B=Multiply(
    A=Multiply(A=@n6.R, B=@n9)
    B=MaterialFunctionCall(Alpha=@n6.R, Min=@n10.Dissolve, Max=@n11) {materialFunction=&SmoothStep}.Result
   )
  )
 ]
}
