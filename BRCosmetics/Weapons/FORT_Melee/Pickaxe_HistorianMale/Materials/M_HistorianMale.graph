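# Pickaxe material for the HistorianMale cosmetic, adding animated glowing runes to a textured lit surface.
# A Time parameter reads a curve atlas row: R grows a pre-skinned XZ sphere mask, G scales emissive.
# FXMask G and B split the reveal into sharp and soft tinted runes, added to the emissive texture.
# Parameter names mislead: NormalMap holds packed specular/metallic/roughness, SpecularMap holds emissive; attributes end in MF_TechArtOverrides_accessories.
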Material /BRCosmetics/Weapons/FORT_Melee/Pickaxe_HistorianMale/Materials/M_HistorianMale
graph {
 properties {
  bUseMaterialAttributes=true
  bUsedWithMeshDeformer=true
  bUsedWithSkeletalMesh=true
  pixelDepthOffsetMode=PDOM_Legacy
  usedShadingModels=Default Lit
 }
 nodes [
  Frac(None=Multiply(A=Time()) {constB=0.5})
  n1 = TextureSampleParameter2D() {
   group=Basic Material Parameters
   parameterName=DiffuseMap
   texture=/BRCosmetics/Weapons/FORT_Melee/Pickaxe_HistorianMale/Textures/T_Historian_Male_D
  }
  n2 = TextureSampleParameter2D() {
   group=Basic Material Parameters
   parameterName=NormalMap
   samplerType=SAMPLERTYPE_Masks
   texture=/BRCosmetics/Weapons/FORT_Melee/Pickaxe_HistorianMale/Textures/T_Historian_Male_S
  }
  n3 = TextureSampleParameter2D() {
   group=Basic Material Parameters
   parameterName=SpecularMap
   texture=/BRCosmetics/Weapons/FORT_Melee/Pickaxe_HistorianMale/Textures/T_Historian_Male_E
  }
  n4 = ComponentMask(None=Constant3Vector() {constant=<rgba 2.1962 0 55.206 0>}) {b=true, r=true}
  n5 = ComponentMask(None=VertexInterpolator(VS=PreSkinnedPosition()).PS) {b=true, r=true}
  n6 = CurveAtlasRowParameter(CurveTime=Saturate(None=ScalarParameter() {group=RunesEmissive, parameterName=Time, sliderMax=1})) {
   atlas=/Game/Weapons/FORT_Melee/Pickaxe_HistorianMale/Textures/HistorianCurveAtlas
   curve=/Game/Weapons/FORT_Melee/Pickaxe_HistorianMale/Textures/HistorianAnimCurve
   desc="R=Speed G=Emissive"
   group=RunesEmissive
   parameterName=AnimCurve
  }
  n7 = Constant()
  n8 = MaterialFunctionCall(
   Input=@n6.R
   Input Low=@n7
   Input High=Constant() {r=1}
   Target Low=@n7
   Target High=ScalarParameter() {defaultValue=23, group=RunesEmissive, parameterName=SphereMaskRadius}
  ) {materialFunction=&RemapValueRange}
  n9 = ScalarParameter() {group=RunesEmissive, parameterName=SphereMaskHardness}
  n10 = TextureSampleParameter2D() {
   group=Basic Material Parameters
   parameterName=FXMask
   samplerType=SAMPLERTYPE_Masks
   texture=/BRCosmetics/Weapons/FORT_Melee/Pickaxe_HistorianMale/Textures/T_Historian_Male_FX
  }
  n11 = Multiply(A=SphereMask(A=@n4, B=@n5, Radius=@n8.Result, Hardness=@n9), B=@n10.G)
  n12 = VectorParameter() {group=RunesEmissive, parameterName=RunesEmissive, defaultValue=<rgba 0 6.1383 7 1>}
  n13 = SphereMask(A=@n4, B=@n5, Radius=@n8.Result, Hardness=@n9)
  n14 = Multiply(A=Multiply(A=@n13, B=@n13), B=@n10.B)
  n15 = VectorParameter() {group=RunesEmissive, parameterName=RunesSoftEmissive, defaultValue=<rgba 0.1021 0.3633 0.4 1>}
  n16 = Multiply(A=Add(A=Multiply(A=@n11, B=@n12.RGB), B=Multiply(A=@n14, B=@n15.RGB)), B=@n6.G)
  n17 = SetMaterialAttributes!(
   MaterialAttributes
   Base Color=NamedRerouteDeclaration(None=@n1.RGB) {name=baseColor, nodeColor=<rgba 0.0781 1 0 1>}
   Metallic=NamedRerouteDeclaration(None=@n2.G) {name=metallic, nodeColor=<rgba 0.0078 0 1 1>}
   Specular=NamedRerouteDeclaration(None=@n2.R) {name=specular, nodeColor=<rgba 1 0 0.1406 1>}
   Roughness=NamedRerouteDeclaration(None=@n2.B) {name=roughness, nodeColor=<rgba 0 1 1 1>}
   Emissive Color=NamedRerouteDeclaration(None=Add(A=@n3.RGB, B=@n16)) {name=emissive, nodeColor=<rgba 1 0.4219 0 1>}
   Normal=NamedRerouteDeclaration(None=TextureSample() {samplerType=SAMPLERTYPE_Normal, texture=/BRCosmetics/Weapons/FORT_Melee/Pickaxe_HistorianMale/Textures/T_Historian_Male_N}.RGB) {name=normal, nodeColor=<rgba 0 1 0.5078 1>}
  ) {attributeSetTypes [
    #1
    #2
    #3
    #4
    #5
    #6
   ]}
  MaterialFunctionCall(In=@n17) {materialFunction=&MF_TechArtOverrides_accessories}
 ]
}
